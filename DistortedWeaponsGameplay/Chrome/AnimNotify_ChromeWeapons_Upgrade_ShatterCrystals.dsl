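; Anim notify for Chrome weapons that signals the crystal-shatter moment of a weapon upgrade animation.
; On notify, it casts the mesh owner to a Fort player pawn and gets its current weapon.
; It tries the Chrome shotgun, then the Chrome assault rifle, calling that weapon's AnimNotifyUpgradeShatterCrystals.
; It returns true when either cast succeeds and false otherwise.
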(fn Received_Notify
  (params (MeshComp :type Skeletal%20Mesh%20Component%20Object%20Reference) (Animation :type Anim%20Sequence%20Base%20Object%20Reference) (EventReference :type Anim%20Notify%20Event%20Reference%20Structure%20%28by%20ref%29))
  (bind _asfort_player_pawn (Utilities|Casting|CastToFortPlayerPawn (Components|GetOwner MeshComp)))
  (bind _asb_shotgun_chrome_athena (Utilities|Casting|CastToB_Shotgun_Chrome_Athena (Weapon|GetCurrentWeapon _asfort_player_pawn))
    (:then
      (Class|BShotgunChromeAthena|AnimNotifyUpgradeShatterCrystals _asb_shotgun_chrome_athena))
    (:CastFailed
      (bind _asb_assault_chrome_athena (Utilities|Casting|CastToB_Assault_Chrome_Athena (Weapon|GetCurrentWeapon _asfort_player_pawn))
        (:then
          (Class|BAssaultChromeAthena|AnimNotifyUpgradeShatterCrystals _asb_assault_chrome_athena))
        (:CastFailed
          (return false)))))
  (return true))
