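; XP component for Chrome multi-item weapons (assault rifle, shotgun) that turns damage into upgrade progress.
; On authority, trigger gameplay events from the held Chrome weapon grant damage-scaled XP at rates chosen by target tags.
; A pending power-up sets a preloaded override fire mode, then reload-tag removal with a full magazine flushes it on dedicated server and plays post-upgrade montages.
; Item switches apply an upgrade-tag effect and report quest stats.
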(event EventBeginPlay
  (bind (_eventtag _instigator _target _optionalobject _optionalobject2 _contexthandle _instigatortags _targettags _eventmagnitude) (Utilities|Struct|BreakGameplayEventData (Variables|Default|GetPayload)))
  (bind _payload (Variables|Default|GetPayload))
  (bind (_eventtag_2 _instigator_2 _target_2 _optionalobject_2 _optionalobject2_2 _contexthandle_2 _instigatortags_2 _targettags_2 _eventmagnitude_2) (Utilities|Struct|BreakGameplayEventData _payload))
  (bind _payload_2 (Variables|Default|GetPayload))
  (bind (_eventtag_3 _instigator_3 _target_3 _optionalobject_3 _optionalobject2_3 _contexthandle_3 _instigatortags_3 _targettags_3) (Utilities|Struct|BreakGameplayEventData _payload_2))
  (bind (_eventtag_4 _instigator_4 _target_4 _optionalobject_4 _optionalobject2_4 _contexthandle_4 _instigatortags_4) (Utilities|Struct|BreakGameplayEventData (Variables|Default|GetPayload)))
  (bind _returnvalue (MultiItem|GetOwningPawn))
  (bind _returnvalue_2 (MultiItem|GetOwningPawn))
  (|CallParentFunction)
  (CallFunction|LoadUpgradeFireMode)
  (CallFunction|WaitforReload)
  (Variables|Default|SetCurrentItemIndex (MultiItem|GetSelectedItemIndex))
  (Utilities|SwitchHasAuthority
    (:Authority
      (Utilities|IsValid _returnvalue_2
        (:Is%20Valid
          (bind (_asynctaskproxy _payload_3) (Ability|Async|WaitGameplayEventToActor _returnvalue_2 (Variables|Default|GetXPTriggerTag) false false)
            (:EventReceived
              (Variables|Default|SetPayload _payload_3)
              (Utilities|IsValid _returnvalue
                (:Is%20Valid
                  (if (and (GameplayTags|HasAnyTags _instigatortags_4 (Variables|Default|GetChromeWeaponTags)) (Utilities|Equal(Object) (Variables|/Script/FortniteGame.FortPawn|GetCurrentWeapon _returnvalue) (MultiItem|GetWeaponActor)))
                    (if (> _eventmagnitude 0)
                      ;[1 0.2327 0.7659] Branch based on type of target.
                      (if (GameplayTags|HasAnyTags _targettags_3 (GameplayTags|MakeLiteralGameplayTagContainer "(GameplayTags=((TagName=\"Pawn.Athena.NPC.Wildlife.Predator.SpicySake\"),(TagName=\"Pawn.Athena.NonParticipant\"),(TagName=\"Pawn.Athena.NPC.Deimos\"),(TagName=\"Pawn.Athena.NPC.CreativeSentry\")))"))
                        ;[0.08121 1 0.1267] Adjusted XP Rates
                        (PowerUp|IncrementItemXP (MultiItem|GetOwningMultiItem) (* _eventmagnitude_2 (ScalableFloat|GetValueatLevel (Variables|Default|GetXpPerDamage_LowRate))))
                        (elif (GameplayTags|HasAnyTags _targettags_3 (GameplayTags|MakeLiteralGameplayTagContainer "(GameplayTags=((TagName=\"Pawn.Athena.NPC.Labrador\"),(TagName=\"Pawn.Athena.NPC.Creative.Guard\")))"))
                          (PowerUp|IncrementItemXP (MultiItem|GetOwningMultiItem) (* _eventmagnitude_2 (ScalableFloat|GetValueatLevel (Variables|Default|GetXpPerDamage_MediumRate))))
                          (elif (GameplayTags|HasAnyTags _targettags_3 (GameplayTags|MakeLiteralGameplayTagContainer "(GameplayTags=((TagName=\"Pawn.Athena.NPC.Wildlife\")))"))
                            (PowerUp|IncrementItemXP (MultiItem|GetOwningMultiItem) (* _eventmagnitude_2 (ScalableFloat|GetValueatLevel (Variables|Default|GetXpPerDamage_HighRate))))
                            (elif (GameplayTags|HasAnyTags _targettags_3 (GameplayTags|MakeLiteralGameplayTagContainer "(GameplayTags=((TagName=\"Pawn.Athena.Player\")))"))
                              (PowerUp|IncrementItemXP (MultiItem|GetOwningMultiItem) (* _eventmagnitude_2 (ScalableFloat|GetValueatLevel (Variables|Default|GetXpPerDamage_PlayerRate)))))))))))))))))))

(event Custom|LoadUpgradeFireMode
  (bind _upgradefiremode (CallFunction|GetUpgradeFireMode))
  ;[0.4498 0.0389 1] Load the upgrade fire mode and store it to prevent garbage collection. This is necessary on client and server to prevent the fire mode from being loaded when it is set, which could cause timing issues with the reload animation.
  (if (Utilities|IsValidSoftObjectReference _upgradefiremode)
    (bind _object (Utilities|AsyncLoadAsset _upgradefiremode)
      (:Completed
        (Variables|Default|SetLoadedFireMode _object)))))

(event MultiItem|PowerUp|EventOnItemXPChanged
  (params (XPChangeAmount :type Float%20%28single-precision%29) (CurrentXP :type Float%20%28single-precision%29))
  ;[0.2131 0.9103 1] XP and level up events
  (Default|CallOnXPChanged XPChangeAmount (MultiItem|PowerUp|GetMultiItemXPPercentage)))

(event Custom|Client_SetOverrideFireMode
  (bind _upgradefiremode (CallFunction|GetUpgradeFireMode))
  (bind _returnvalue (MultiItem|GetWeaponActor))
  ;[0.2131 0.9103 1] XP and level up events
  (Utilities|IsValid _returnvalue
    (:Is%20Valid
      (if (Utilities|IsValidSoftObjectReference _upgradefiremode)
        (FortWeapon|SetOverrideFireModeData _returnvalue _upgradefiremode)))))

(event MultiItem|PowerUp|EventOnPowerUpPending
  (params (CurrentItemIndex :type Integer) (PendingItemIndex :type Integer))
  (bind _upgradefiremode (CallFunction|GetUpgradeFireMode))
  (bind _returnvalue (MultiItem|GetWeaponActor))
  ;[0.2131 0.9103 1] XP and level up events
  (Default|CallOnPowerUpReady CurrentItemIndex PendingItemIndex)
  (CallFunction|ClientSetOverrideFireMode)
  (Utilities|IsValid _returnvalue
    (:Is%20Valid
      (if (Utilities|IsValidSoftObjectReference _upgradefiremode)
        (FortWeapon|SetOverrideFireModeData _returnvalue _upgradefiremode)))))

(event EventOnItemSwitched
  (params (CurrentItemIndex :type Integer))
  ;[0.2131 0.9103 1] XP and level up events
  (Default|CallOnItemUpgraded CurrentItemIndex))

(event MultiItem|EventOnItemSwitchedAwayFrom
  (params (CurrentItemIndex :type Integer))
  (bind _returnvalue (MultiItem|GetOwningPawn))
  (bind _returnvalue_2 (Fortnite|GetFortPlayerControllerfromActor _returnvalue))
  ;[0.2131 0.9103 1] XP and level up events
  (GameplayEffects|ApplyGameplayEffectToSelf :self (Ability|GetAbilitySystemComponent (MultiItem|GetOwningPawn)) :GameplayEffectClass "/DistortedWeaponsGameplay/Chrome/GE_ChromeWeapons_UpgradeTag_Grant.GE_ChromeWeapons_UpgradeTag_Grant_C" :Level 0.0)
  (Default|CallUpgradeSwitch)
  (for _ (range 1 (+ (- (MultiItem|GetSelectedItemIndex) (Variables|Default|GetCurrentItemIndex)) 1))
    (Utilities|IsValid _returnvalue_2
      (:Is%20Valid
        (Quest|SendComplexCustomStatEvent (Quests|GetQuestManager _returnvalue_2 "Invalid") 0 "(GameplayTags=((TagName=\"Granted.DistortedWeapons.ChromeWeapon.UpgradeWeapon\")))" "(GameplayTags=)" 1)))))

(event Custom|WaitForReload
  (bind _returnvalue (MultiItem|GetOwningPawn))
  (bind _returnvalue_2 (Weapon|GetCurrentWeapon _returnvalue))
  (bind _returnvalue_3 (MultiItem|GetOwningPawn))
  (bind _reloadtagremovedasyncaction (Variables|Default|GetReloadTagRemovedAsyncAction))
  ;[0.2843 0.1367 1] Upgrade On Reload
  (Utilities|IsValid _returnvalue_3
    (:Is%20Valid
      (Ability|Tasks|WaitGameplayTagAddToActor _returnvalue_3 "(TagName=\"Gameplay.Action.Player.Reload\")"
        (:Added
          (Utilities|IsValid _returnvalue
            (:Is%20Valid
              (Utilities|IsValid _returnvalue_2
                (:Is%20Valid
                  (if (== _returnvalue_2 (MultiItem|GetWeaponActor))
                    (bind _asynctaskproxy (Ability|Tasks|WaitGameplayTagRemoveFromActor _returnvalue "(TagName=\"Gameplay.Action.Player.Reload\")" true)
                      (:then
                        (Variables|Default|SetReloadTagRemovedAsyncAction _asynctaskproxy))
                      (:Removed
                        (CallFunction|HandleOwnerReloaded)
                        (Utilities|IsValid _reloadtagremovedasyncaction
                          (:Is%20Valid
                            (AsyncAction|Cancel _reloadtagremovedasyncaction)))))))))))))))

(event Custom|BroadcastPlayUpgradeMontage
  (bind _returnvalue (MultiItem|GetOwningPawn))
  (bind _returnvalue_2 (MultiItem|GetOwningPawn))
  ;[0.2843 0.1367 1] Upgrade On Reload
  (Utilities|IsValid _returnvalue_2
    (:Is%20Valid
      (if (not (Pawn|IsLocallyControlled _returnvalue_2))
        (Utilities|IsValid _returnvalue
          (:Is%20Valid
            (Animation|Montage|PlayMontage (Variables|/Script/Engine.Character|GetMesh _returnvalue) (select (Tags|WeaponDataHasMatchingGameplayTag (Weapon|GetCurrentWeapon _returnvalue) "(TagName=\"Item.Weapon.Ranged.Assault.Chrome.Distorted\")") (Variables|Default|GetARPostUpgradeMontage) (Variables|Default|GetShotgunPostUpgradeMontage)) 1.0 0.0 (Variables|Default|GetPostUpgradeSectionName))))))))

(event Custom|HandleOwnerReloaded
  (bind _returnvalue (MultiItem|GetOwningPawn))
  (bind _returnvalue_2 (MultiItem|GetOwningPawn))
  (bind _returnvalue_3 (Weapon|GetCurrentWeapon _returnvalue))
  ;[0.2843 0.1367 1] Upgrade On Reload
  (if (MultiItem|PowerUp|IsPowerUpPending)
    (Utilities|IsValid _returnvalue
      (:Is%20Valid
        (Utilities|IsValid _returnvalue_3
          (:Is%20Valid
            (if (== (FortWeapon|GetBulletsPerClip _returnvalue_3) (FortWeapon|GetMagazineAmmoCount _returnvalue_3))
              (if (Networking|IsDedicatedServer)
                (MultiItem|PowerUp|FlushPendingPowerUp)
                (CallFunction|BroadcastPlayUpgradeMontage)
                (else
                  _))
              (Utilities|IsValid _returnvalue_2
                (:Is%20Valid
                  (Animation|Montage|PlayMontage (Variables|/Script/Engine.Character|GetMesh _returnvalue_2) (select (Tags|WeaponDataHasMatchingGameplayTag (Weapon|GetCurrentWeapon _returnvalue_2) "(TagName=\"Item.Weapon.Ranged.Assault.Chrome.Distorted\")") (Variables|Default|GetARPostUpgradeMontage) (Variables|Default|GetShotgunPostUpgradeMontage)) 1.0 0.0 (Variables|Default|GetPostUpgradeSectionName)))))))))))

(fn GetUpgradeFireMode
  (bind _returnvalue (MultiItem|GetWeaponActor))
  (Utilities|IsValid _returnvalue
    (:Is%20Valid
      (if (Tags|WeaponDataHasMatchingGameplayTag _returnvalue "(TagName=\"Item.Weapon.Ranged.Assault.Chrome.Distorted\")")
        (return (Variables|Default|GetAssaultUpgradeFireMode))
        (elif (Tags|WeaponDataHasMatchingGameplayTag _returnvalue "(TagName=\"Item.Weapon.Ranged.Shotgun.Chrome.Distorted\")")
          (return (Variables|Default|GetShotgunUpgradeFireMode)))))
    (:Is%20Not%20Valid
      _))
  (return 0))
