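; Audio component for Chrome weapons that plays upgrade-ready and evolve sounds.
; On BeginPlay it caches the owner's Quartz gunfire component, unused here, and binds to the Chrome XP component's upgrade events.
; Upgrade-ready plays a 2D sound only for the locally controlled instigator.
; Upgrades and switches spawn an attached evolve sound, first-person for the local player, only when selected item index plus upgrade increment is 2 or 3.
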(event EventBeginPlay
  (bind _self self)
  (bind _returnvalue (Components|GetOwner _self))
  (Variables|Default|SetQuartzGunfireComponent (Actor|GetComponentbyClass _returnvalue "/Script/FortniteGame.FortQuartzGunfireComponent"))
  (bind _output_get (Variables|Default|SetChromeWeaponXPComponent (Actor|GetComponentbyClass _returnvalue "/DistortedWeaponsGameplay/Chrome/C_ChromeWeapons_XPComponent.C_ChromeWeapons_XPComponent_C")))
  (Utilities|IsValid _output_get
    (:Is%20Valid
      (Default|BindEventtoOnItemUpgraded _output_get (EventDispatchers|CreateEvent :!event Custom|ItemUpgraded))
      (Default|BindEventtoOnPowerUpReady _output_get (EventDispatchers|CreateEvent :!event Custom|UpgradeReady))
      (Default|AssignUpgradeSwitch _output_get (EventDispatchers|CreateEvent :!event Custom|UpgradeSwitch_Event)))))

(event Custom|UpgradeReady
  (params (CurrentItemIndex :type Integer) (PendingItemIndex :type Integer))
  (Networking|CanExecuteCosmeticEvents
    (:True
      (if (Pawn|IsLocallyControlled (Variables|/Script/Engine.Actor|GetInstigator (Components|GetOwner self)))
        (Audio|PlaySound2D (Variables|Default|GetSFX_UpgradeReady))))))

(event Custom|UpgradeSwitch_Event
  ;[1 1 1] Might be a better way to do this. However, this works because it relies on the creation of new gun after upgrade swap..
  (Variables|Default|SetUpgradeIncrement 1)
  (CallFunction|ItemUpgraded 0))

(event Custom|ItemUpgraded
  (params (CurrentItemIndex :type Integer))
  (Variables|Default|SetUpgradeIncrement CurrentItemIndex)
  (CallFunction|ItemUpgradedServer))

(event Custom|ItemUpgradedServer
  (Networking|CanExecuteCosmeticEvents
    (:True
      (if (Utilities|Select :Index (+ (MultiItem|GetSelectedItemIndex (Variables|Default|GetChromeWeaponXPComponent)) (Variables|Default|GetUpgradeIncrement)) :Option%200 false :Option%201 false :Option%202 true :Option%203 true :Option%204 false :Option%205 false)
        (Audio|SpawnSoundAttached (select (Pawn|IsLocallyControlled (Variables|/Script/Engine.Actor|GetInstigator (Components|GetOwner self))) (Variables|Default|GetEvolve_1P) (Variables|Default|GetEvolve)) (Variables|/Script/Engine.Actor|GetRootComponent (Components|GetOwner)))))))
