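; Pickup Blueprint for Chrome items that adds a Chrome visual effect when the pickup is tossed.
; On the tossed event it runs the parent logic, then spawns the NS_Weapon_Pickup_Chrome Niagara system attached and snapped to its root component.
; The construction script only calls the parent, so all other behavior is inherited.
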(event EventOnTossed
  (|CallParentFunction)
  (Niagara|SpawnSystemAttached "/DistortedWeaponsGameplay/Chrome/FX/NS_Weapon_Pickup_Chrome.NS_Weapon_Pickup_Chrome" (Variables|Transformation|GetRootComponent) "None" "0, 0, 0" "0, 0, 0" "SnapToTarget"))

(fn ConstructionScript
  (|CallParentFunction))
