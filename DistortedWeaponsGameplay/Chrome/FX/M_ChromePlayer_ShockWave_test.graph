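# Translucent unlit Niagara sprite material, apparently a test shockwave effect for the Chrome player, with emissive color and distance-scaled refraction.
# Radial UVs distorted by panning Perlin noise sample a plasma texture and position a curve-atlas ring band, moved by Dynamic Parameter 1 or a sine of time.
# Band and plasma form opacity with a masked-out center, particle alpha and depth fade.
# Band position also hue-shifts the particle color.
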Material /DistortedWeaponsGameplay/Chrome/FX/M_ChromePlayer_ShockWave_test
graph {
 properties {
  bUsedWithNiagaraSprites=true
  blendMode=BLEND_Translucent
  pixelDepthOffsetMode=PDOM_Legacy
  refractionMethod=RM_IndexOfRefraction
  shadingModel=MSM_Unlit
  twoSided=true
  usedShadingModels=Unlit
 }
 nodes [
  n1 = TextureCoordinate()
  n2 = Constant() {r=0.5}
  n3 = ScalarParameter() {defaultValue=0.1, parameterName=Center}
  n4 = Constant() {r=0.2}
  n5 = ParticleColor()
  n6 = TextureObjectParameter() {
   group=Texture
   parameterName=Texture_Texture
   samplerType=SAMPLERTYPE_LinearGrayscale
   sortPriority=0
   texture=&/T_Plasma_B_Greyscale_HalfWidth_2
  }
  n7 = VectorParameter() {group=Texture, parameterName=Texture_UVControl, channelNames {a=VPan, b=UPan, g=VScale, r=UScale}, defaultValue=<rgba 1 -0.5 0 0.1>}
  n8 = TextureObjectParameter() {
   group=Noise
   parameterName=Noise_Texture
   samplerType=SAMPLERTYPE_LinearGrayscale
   sortPriority=0
   texture=&T_SimplePerlin_Greyscale
  }
  n9 = VectorParameter() {group=Noise, parameterName=Noise_UVControl, channelNames {a=VPan, b=UPan, g=VScale, r=UScale}, defaultValue=<rgba 1 1 0 0.3>}
  n10 = NamedRerouteUsage()
  n11 = MaterialFunctionCall(Vector or UVs=Add(A=TextureCoordinate()) {constB=-0.5}) {materialFunction=&VectorToRadialValue}
  n12 = VectorParameter() {
   group=Noise
   parameterName=Noise_NoiseControl
   sortPriority=60
   channelNames {a=VStrength, b=UStrength, g=VOffset, r=UOffset}
   defaultValue=<rgba -0.5 -0.5 0 0.2>
  }
  n13 = Time()
  n14 = LinearInterpolate(A=ConstantBiasScale(None=Sine(None=@n13)), B=DynamicParameter().Param1, Alpha=ScalarParameter() {defaultValue=1, parameterName=UseP})
  n15 = Add(
   A=MaterialFunctionCall!(
    NoiseValue=MaterialFunctionCall(Texture=@n8, UVControl_RGB=@n9, UVControl_A=@n9, "[Override]TimeInput"=@n10, "[Override]UVInput"=@n11.Radial Coordinates) {materialFunction=&FXCommon_TextureSample}.RG
    Control_RGB=@n12
    Control_A=Multiply(A=@n12.VStrength, B=@n14)
   ) {materialFunction=/Game/Effects/Niagara/MasterMaterials/Functions/Common/FXCommon_NoiseControl}.NoiseValue
   B=@n11.Radial Coordinates
  )
  n16 = MaterialFunctionCall(
   Texture=@n6
   UVControl_RGB=@n7
   UVControl_A=@n7
   "[Override]TimeInput"=NamedRerouteUsage()
   "[Override]UVInput"=LinearInterpolate(A=TextureCoordinate(), B=@n15, Alpha=ScalarParameter() {defaultValue=1, parameterName=Method})
  ) {materialFunction=&FXCommon_TextureSample}
  n17 = Multiply(A=@n14, B=ScalarParameter() {defaultValue=-0.5, parameterName=Flow})
  n18 = ScalarParameter() {defaultValue=0.35, parameterName=FlowOffset}
  n19 = Multiply(
   A=Add(
    A=MaterialFunctionCall(Float2=@n15) {materialFunction=&BreakOutFloat2Components}.G
    B=Add(A=@n17, B=@n18)
   )
   B=ScalarParameter() {defaultValue=3, parameterName=Widht}
  )
  n20 = CurveAtlasRowParameter(CurveTime=@n19) {
   atlas=&/CLA_Shockwave
   curve=&/CL_Shockwave_Color1
   defaultValue=1
   parameterName=Curve
  }
  n21 = MaterialFunctionCall(X=Constant() {r=0.5}, Y=@n19) {materialFunction=&MakeFloat2}
  n22 = Constant() {r=1}
  n23 = Multiply(
   A=Multiply(
    A=@n20.R
    B=MaterialFunctionCall!(TexCoord=@n21.Result, ClampChannel=@n22) {materialFunction=/Game/Effects/Fort_Effects/MaterialFunctions/MF_FX_TilingMethod}.Result
   )
   B=ScalarParameter() {defaultValue=2, parameterName=TextureStr}
  )
  n24 = Power(Base=Add(A=Subtract(A=@n16.R), B=@n23))
  n25 = Power(Base=@n16.R, Exp=ScalarParameter() {defaultValue=15, parameterName=OpacityRefine_P1})
  n26 = ScalarParameter() {defaultValue=50, parameterName=OpacityRefine_M1}
  n27 = ScalarParameter() {defaultValue=2, parameterName=OpacityRefine_M2}
  n28 = Multiply(A=@n23, B=LinearInterpolate(A=Multiply(A=@n25, B=@n26), B=Multiply(A=@n16.R, B=@n27)))
  n29 = ScalarParameter() {defaultValue=1, parameterName=Method}
  n30 = Multiply(A=OneMinus(None=SphereMask(A=@n1, B=@n2, Radius=@n3, Hardness=@n4)), B=Multiply(A=@n5.A, B=LinearInterpolate(A=@n24, B=@n28, Alpha=@n29)))
  DepthFade(Opacity=@n30, FadeDistance=ScalarParameter() {parameterName=DFDistance})
  n31 = CameraPositionWS()
  n32 = ParticlePositionWS()
  n33 = VectorParameter() {parameterName=RefRemap, defaultValue=<rgba 400 2048 0 1>}
  n34 = MaterialFunctionCall(
   Input=MaterialFunctionCall(Vector 3=Subtract(A=@n31, B=@n32)) {materialFunction=/Engine/Functions/Engine_MaterialFunctions02/Utility/VectorLength}.V3 Length
   Input Low=@n33.R
   Input High=@n33.R
  ) {materialFunction=&RemapValue_0-1}
  LinearInterpolate(B=LinearInterpolate(A=ScalarParameter() {defaultValue=1.75, parameterName=RefS}, Alpha=Saturate(None=@n34.Result)), Alpha=@n30) {constA=1}
  n35 = Add(
   A=Multiply(A=@n19, B=ScalarParameter() {defaultValue=0.6, parameterName=Scale})
   B=ScalarParameter() {defaultValue=-0.1, parameterName=Offset}
  )
  MaterialFunctionCall(Emissive=Multiply(
    A=VectorParameter() {parameterName=PostColor, defaultValue=<rgba 1 1 1 1>}.RGB
    B=MaterialFunctionCall(Hue Shift Percentage=@n35, Texture=@n5.RGB) {materialFunction=&HueShift}.Result
   )) {materialFunction=&MF_FX_EmissiveAdjustments}
  NamedRerouteDeclaration(None=LinearInterpolate(A=Time() {bOverride_Period=true, period=600}, B=DynamicParameter().Param3, Alpha=ScalarParameter() {parameterName=UseDPTimer})) {name=Timerr, nodeColor=<rgba 0 0.2266 1 1>}
  Normalize(VectorInput=Subtract(A=@n2, B=@n1))
  TextureCoordinate()
 ]
}
