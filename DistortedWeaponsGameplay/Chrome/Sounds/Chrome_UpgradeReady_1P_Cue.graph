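# First-person sound cue for the Chrome weapon "upgrade ready" sound in Distorted Weapons gameplay.
# A single wave player feeds the root directly, playing one wave at reduced volume through the weapon foley sound class.
# Overridden concurrency allows one instance per owner and stops the oldest when a new one starts.
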SoundCue /DistortedWeaponsGameplay/Chrome/Sounds/Chrome_UpgradeReady_1P_Cue
root=WavePlayer_7
nodes [{id=WavePlayer_7, properties {soundWaveAssetPtr=&/Chrome_UpgradeReady_10}}]
properties {
 bOverrideConcurrency=true
 maxDistance=2097152
 soundClassObject=&Fort_WeaponFoley
 volumeMultiplier=0.376
 attenuationOverrides {falloffDistance=1600}
 concurrencyOverrides {bLimitToOwner=true, maxCount=1, resolutionRule=StopOldest}
}
