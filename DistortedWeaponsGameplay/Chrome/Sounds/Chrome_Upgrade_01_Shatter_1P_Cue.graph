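# First-person sound cue for a Chrome weapon upgrade, layering a shatter sound with a rarity-up sound under the weapon foley sound class.
# A mixer plays both layers together, with the shatter louder, each picking randomly from four equally weighted waves.
# Modulators vary the shatter's pitch but hold the rarity-up layer's pitch fixed.
# Concurrency override allows one instance per owner, stopping the oldest.
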SoundCue /DistortedWeaponsGameplay/Chrome/Sounds/Chrome_Upgrade_01_Shatter_1P_Cue
root=Mixer_0
nodes [
 {id=Mixer_0, children [Modulator_14,Modulator_0], properties {inputVolume [1.1,0.45]}}
 {id=Modulator_0, children [Random_1], properties {pitchMax=1, pitchMin=1}}
 {id=Modulator_14, children [Random_0], properties {pitchMax=1.1}}
 {id=Random_0, children [WavePlayer_0,WavePlayer_1,WavePlayer_3,WavePlayer_4], properties {weights [1,1,1,1]}}
 {id=Random_1, children [WavePlayer_5,WavePlayer_6,WavePlayer_7,WavePlayer_8], properties {weights [1,1,1,1]}}
 {id=WavePlayer_0, properties {soundWaveAssetPtr=&/Chrome_Upgrade_01_Shatter_01}}
 {id=WavePlayer_1, properties {soundWaveAssetPtr=&/Chrome_Upgrade_01_Shatter_02}}
 {id=WavePlayer_3, properties {soundWaveAssetPtr=&/Chrome_Upgrade_01_Shatter_03}}
 {id=WavePlayer_4, properties {soundWaveAssetPtr=&/Chrome_Upgrade_01_Shatter_04}}
 {id=WavePlayer_5, properties {soundWaveAssetPtr=&/Chrome_Upgrade_02_RarityUp_01}}
 {id=WavePlayer_6, properties {soundWaveAssetPtr=&/Chrome_Upgrade_02_RarityUp_02}}
 {id=WavePlayer_7, properties {soundWaveAssetPtr=&/Chrome_Upgrade_02_RarityUp_03}}
 {id=WavePlayer_8, properties {soundWaveAssetPtr=&/Chrome_Upgrade_02_RarityUp_04}}
]
properties {
 bOverrideConcurrency=true
 maxDistance=2097152
 soundClassObject=&Fort_WeaponFoley
 volumeMultiplier=0.752
 attenuationOverrides {falloffDistance=1600}
 concurrencyOverrides {bLimitToOwner=true, maxCount=1, resolutionRule=StopOldest}
}
