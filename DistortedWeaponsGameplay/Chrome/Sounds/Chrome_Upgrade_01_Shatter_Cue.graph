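# Sound cue for the shatter moment of a Chrome weapon upgrade, classed as weapon foley.
# A mixer layers one of four random shatter waves, pitch-varied by a modulator, over a quieter random pick of four rarity-up waves at fixed pitch.
# Playback uses the Chrome upgrade 3P attenuation asset and shared local and global foley concurrency sets.
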SoundCue /DistortedWeaponsGameplay/Chrome/Sounds/Chrome_Upgrade_01_Shatter_Cue
root=Mixer_1
nodes [
 {id=Mixer_1, children [Modulator_2,Modulator_0], properties {inputVolume [1.1,0.45]}}
 {id=Modulator_0, children [Random_3], properties {pitchMax=1, pitchMin=1}}
 {id=Modulator_2, children [Random_0], properties {pitchMax=1.1}}
 {id=Random_0, children [WavePlayer_0,WavePlayer_1,WavePlayer_3,WavePlayer_4], properties {weights [1,1,1,1]}}
 {id=Random_3, children [WavePlayer_12, WavePlayer_13, WavePlayer_14, WavePlayer_15], properties {weights [1,1,1,1]}}
 {id=WavePlayer_0, properties {soundWaveAssetPtr=&/Chrome_Upgrade_01_Shatter_01}}
 {id=WavePlayer_1, properties {soundWaveAssetPtr=&/Chrome_Upgrade_01_Shatter_02}}
 {id=WavePlayer_12, properties {soundWaveAssetPtr=&/Chrome_Upgrade_02_RarityUp_01}}
 {id=WavePlayer_13, properties {soundWaveAssetPtr=&/Chrome_Upgrade_02_RarityUp_02}}
 {id=WavePlayer_14, properties {soundWaveAssetPtr=&/Chrome_Upgrade_02_RarityUp_03}}
 {id=WavePlayer_15, properties {soundWaveAssetPtr=&/Chrome_Upgrade_02_RarityUp_04}}
 {id=WavePlayer_3, properties {soundWaveAssetPtr=&/Chrome_Upgrade_01_Shatter_03}}
 {id=WavePlayer_4, properties {soundWaveAssetPtr=&/Chrome_Upgrade_01_Shatter_04}}
]
properties {
 attenuationSettings=&/Chrome_Upgrade_3P_Att
 maxDistance=2200
 soundClassObject=&Fort_WeaponFoley
 volumeMultiplier=0.626
 attenuationOverrides {falloffDistance=1600}
 concurrencyOverrides {bLimitToOwner=true, maxCount=2, resolutionRule=StopOldest}
 concurrencySet [
  /Game/Sounds/Fort_Foley_Sounds/FoleyLibraries/Settings/Foley_Basic_Local_Conc
  /Game/Sounds/Fort_Foley_Sounds/FoleyLibraries/Settings/Foley_Basic_Global_Conc
 ]
}
