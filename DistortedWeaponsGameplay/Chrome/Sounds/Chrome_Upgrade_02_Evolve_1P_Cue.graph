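# Sound cue for the first-person Chrome weapon upgrade "Evolve" sound (Upgrade 02), in the weapon foley sound class.
# A single wave player feeds the root directly with one sound wave at reduced volume, with no randomization or modulation.
# Overridden concurrency allows one instance per owner and stops the oldest when a new one starts.
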SoundCue /DistortedWeaponsGameplay/Chrome/Sounds/Chrome_Upgrade_02_Evolve_1P_Cue
root=WavePlayer_0
nodes [{id=WavePlayer_0, properties {soundWaveAssetPtr=&/Chrome_Upgrade_02_Evolve_01}}]
properties {
 bOverrideConcurrency=true
 maxDistance=2097152
 soundClassObject=&Fort_WeaponFoley
 volumeMultiplier=0.601
 attenuationOverrides {falloffDistance=1600}
 concurrencyOverrides {bLimitToOwner=true, maxCount=1, resolutionRule=StopOldest}
}
