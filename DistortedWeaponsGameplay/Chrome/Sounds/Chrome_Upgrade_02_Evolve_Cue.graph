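# Sound cue for a Chrome weapon upgrade "Evolve" sound in the DistortedWeaponsGameplay plugin, routed through the weapon foley sound class.
# A single wave player feeds the root directly, playing one evolve wave with no randomization or modulation.
# It uses the third-person Chrome upgrade attenuation at reduced volume.
# Local and global basic foley concurrency sets govern how many instances play at once.
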SoundCue /DistortedWeaponsGameplay/Chrome/Sounds/Chrome_Upgrade_02_Evolve_Cue
root=WavePlayer_0
nodes [{id=WavePlayer_0, properties {soundWaveAssetPtr=&/Chrome_Upgrade_02_Evolve_01}}]
properties {
 attenuationSettings=&/Chrome_Upgrade_3P_Att
 maxDistance=2200
 soundClassObject=&Fort_WeaponFoley
 volumeMultiplier=0.551
 attenuationOverrides {falloffDistance=1600}
 concurrencyOverrides {bLimitToOwner=true, maxCount=2, resolutionRule=StopOldest}
 concurrencySet [
  /Game/Sounds/Fort_Foley_Sounds/FoleyLibraries/Settings/Foley_Basic_Local_Conc
  /Game/Sounds/Fort_Foley_Sounds/FoleyLibraries/Settings/Foley_Basic_Global_Conc
 ]
}
