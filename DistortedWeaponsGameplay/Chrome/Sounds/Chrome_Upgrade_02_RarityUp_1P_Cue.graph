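# Sound cue named for the first-person Chrome weapon rarity-up upgrade sound, routed to the weapon foley sound class.
# It has no nodes and no root, so as shown it plays no audio and holds only playback settings.
# Those settings lower the volume, override attenuation falloff and concurrency, and allow one instance per owner, with a new play stopping the oldest.
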SoundCue /DistortedWeaponsGameplay/Chrome/Sounds/Chrome_Upgrade_02_RarityUp_1P_Cue
root=null
nodes []
properties {
 bOverrideConcurrency=true
 maxDistance=2097152
 soundClassObject=&Fort_WeaponFoley
 volumeMultiplier=0.501
 attenuationOverrides {falloffDistance=1600}
 concurrencyOverrides {bLimitToOwner=true, maxCount=1, resolutionRule=StopOldest}
}
