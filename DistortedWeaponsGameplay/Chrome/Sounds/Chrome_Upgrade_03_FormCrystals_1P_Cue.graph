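# First-person sound cue for the crystal-forming stage (step 03) of the Chrome weapon upgrade, routed through the weapon foley sound class.
# A fixed short delay at the root feeds a Random node that picks one of four equally weighted FormCrystals wave variations.
# Overridden concurrency allows one instance per owner and stops the oldest when a new one starts.
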SoundCue /DistortedWeaponsGameplay/Chrome/Sounds/Chrome_Upgrade_03_FormCrystals_1P_Cue
root=Delay_0
nodes [
 {id=Delay_0, children [Random_1], properties {delayMax=0.35, delayMin=0.35}}
 {id=Random_1, children [WavePlayer_6,WavePlayer_7,WavePlayer_8,WavePlayer_9], properties {weights [1,1,1,1]}}
 {id=WavePlayer_6, properties {soundWaveAssetPtr=&/Chrome_Upgrade_03_FormCrystals_01}}
 {id=WavePlayer_7, properties {soundWaveAssetPtr=&/Chrome_Upgrade_03_FormCrystals_02}}
 {id=WavePlayer_8, properties {soundWaveAssetPtr=&/Chrome_Upgrade_03_FormCrystals_03}}
 {id=WavePlayer_9, properties {soundWaveAssetPtr=&/Chrome_Upgrade_03_FormCrystals_04}}
]
properties {
 bOverrideConcurrency=true
 maxDistance=2097152
 soundClassObject=&Fort_WeaponFoley
 volumeMultiplier=0.501
 attenuationOverrides {falloffDistance=1600}
 concurrencyOverrides {bLimitToOwner=true, maxCount=1, resolutionRule=StopOldest}
}
