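# Sound cue for the crystal-forming sound of a Chrome weapon upgrade, judging by its name, classed as weapon foley.
# A fixed short delay feeds a Random node that picks one of four equally weighted wave variations.
# A shared 3P attenuation asset spatializes it at reduced volume, and local and global basic foley concurrency sets limit simultaneous playback.
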SoundCue /DistortedWeaponsGameplay/Chrome/Sounds/Chrome_Upgrade_03_FormCrystals_Cue
root=Delay_0
nodes [
 {id=Delay_0, children [Random_1], properties {delayMax=0.35, delayMin=0.35}}
 {id=Random_1, children [WavePlayer_6,WavePlayer_7,WavePlayer_8,WavePlayer_9], properties {weights [1,1,1,1]}}
 {id=WavePlayer_6, properties {soundWaveAssetPtr=&/Chrome_Upgrade_03_FormCrystals_01}}
 {id=WavePlayer_7, properties {soundWaveAssetPtr=&/Chrome_Upgrade_03_FormCrystals_02}}
 {id=WavePlayer_8, properties {soundWaveAssetPtr=&/Chrome_Upgrade_03_FormCrystals_03}}
 {id=WavePlayer_9, properties {soundWaveAssetPtr=&/Chrome_Upgrade_03_FormCrystals_04}}
]
properties {
 attenuationSettings=&/Chrome_Upgrade_3P_Att
 maxDistance=2200
 soundClassObject=&Fort_WeaponFoley
 volumeMultiplier=0.451
 attenuationOverrides {falloffDistance=1600}
 concurrencyOverrides {bLimitToOwner=true, maxCount=2, resolutionRule=StopOldest}
 concurrencySet [
  /Game/Sounds/Fort_Foley_Sounds/FoleyLibraries/Settings/Foley_Basic_Local_Conc
  /Game/Sounds/Fort_Foley_Sounds/FoleyLibraries/Settings/Foley_Basic_Global_Conc
 ]
}
